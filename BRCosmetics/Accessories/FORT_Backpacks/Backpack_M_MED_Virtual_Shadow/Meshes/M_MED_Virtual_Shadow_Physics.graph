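# Physics asset for the Virtual Shadow backpack cosmetic, defining collision bodies and joints for the pack and its two swords.
# Long, thin boxes cover the left and right dyn_sword bones, a capsule covers the pack, and a tiny box sits on root.
# All three parts constrain directly to root: the pack rotates freely, while each sword has moderate swing limits and a much tighter twist limit.
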PhysicsAsset /BRCosmetics/Accessories/FORT_Backpacks/Backpack_M_MED_Virtual_Shadow/Meshes/M_MED_Virtual_Shadow_Physics
bodies {
 dyn_sword_l [{shapeType=Box, extentX=56.7735, extentY=5.0096, extentZ=16.686, center=<1.5696 -0.5951 0.4037>, rotation=<rot 0 0.91 0>}]
 dyn_sword_r [{shapeType=Box, extentX=56.7735, extentY=5.0096, extentZ=16.686, center=<1.5696 -0.4771 -0.4037>, rotation=<rot 0 1.0804 0>}]
 pack [{shapeType=Capsule, length=0.4608, radius=7.2052, center=<0 -6.3296 3.1853>, rotation=<rot 90 0 0>}]
 root [{shapeType=Box, extentX=1.01, extentY=1.01, extentZ=1.01}]
}
constraints [
 {
  bone1Name=pack
  bone2Name=root
  swing1LimitDegrees=45
  swing1Motion=Free
  swing2LimitDegrees=45
  swing2Motion=Free
  twistLimitDegrees=45
  twistMotion=Free
 }
 {bone1Name=dyn_sword_l, bone2Name=root, swing1LimitDegrees=30, swing2LimitDegrees=30, twistLimitDegrees=5}
 {bone1Name=dyn_sword_r, bone2Name=root, swing1LimitDegrees=30, swing2LimitDegrees=30, twistLimitDegrees=5}
]
